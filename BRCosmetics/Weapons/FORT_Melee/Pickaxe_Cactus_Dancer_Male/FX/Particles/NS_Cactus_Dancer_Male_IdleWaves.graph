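# Niagara idle effect for the Cactus Dancer Male pickaxe that pulses mesh waves and a light with Fortnite event audio.
# Both emitters attach particles to the user skeletal mesh and smooth a clamped audio amplitude into an Audiogate value.
# Waves adds Audiogate to engine time for a dynamic material parameter and uses it to boost color and alpha.
# IdleLight scales radius and a looping orange-green-blue color curve by Audiogate, with a minimum amplitude floor.
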NiagaraSystem /BRCosmetics/Weapons/FORT_Melee/Pickaxe_Cactus_Dancer_Male/FX/Particles/NS_Cactus_Dancer_Male_IdleWaves
dependencies {usedDataInterfaces [&NiagaraDataInterfaceColorCurve], usedDynamicInputs [], usedModules [&SystemState]}
emitters {
 IdleLight {
  emitterSpawnScript [/SetVariables() {isSetParameters=true, module=SetVariables_#1}]
  emitterUpdateScript [
   &EmitterState(Life Cycle Mode=::NewEnumerator1, Loop Behavior=::NewEnumerator1, Loop Duration Mode=::NewEnumerator1, UseLoopDelay=false)
   &SpawnBurst_Instantaneous(Use Spawn Probability=false, UseLoopCountLimit=false)
   /SetVariables(Emitter.Audiogate:NiagaraFloat=/Niagara/DynamicInputs/Lerp/SmoothLerpOverTime_Float(
     Convergence Rate:NiagaraFloat=100
     Smooth Value:NiagaraFloat=$Emitter.IncomingAudio
     Target Value:NiagaraFloat=&ClampFloat(Float=&Multiply_Float(A=$NPC.FortniteNPC.EventAudio_Amplitude, B=1), Min=0.6)
    )) {isSetParameters=true, module=SetVariables_#2}
  ]
  particleSpawnScript [&InitializeParticle(Lifetime=3, UsePositionOffset=false, Write Material Random=true)]
  particleUpdateScript [
   &ParticleState(
    Kill Particles When Lifetime Has Elapsed=false
    Let Infinitely Lived Particles Die When Emitter Deactivates=true
    Loop Particles Lifetime=true
   )
   &SkeletalMeshLocation(MaskSpawnGroup=false, Mesh Sampling Type=::NewEnumerator5, Skeletal Mesh=$User.SkeletalMesh, Socket Sampling Mode=::NewEnumerator5)
   &ScaleColor(
    Scale RGB=&MakeVectorFromLinearColorRGB(Color=&MultiplyLinearColor(
      Linear Color=&ColorFromCurve(
       ColorCurve=DataInterface() {propertyValues {
         AlphaCurve {keys [<at 0 1 RCIM_Cubic>]}
         BlueCurve {keys [
           <at 0 0.1922 RCIM_Cubic>
           {arriveTangent=1.6047, interpMode=RCIM_Cubic, leaveTangent=1.6047, time=0.259, value=0.4084}
           <at 0.5034 1>
           <at 0.7624 0.4084>
           <at 1 0.1922>
          ]}
         GreenCurve {keys [
           <at 0 0.6 RCIM_Cubic>
           {arriveTangent=0.2883, interpMode=RCIM_Cubic, leaveTangent=0.2883, time=0.259, value=1}
           <at 0.5034 0.7451>
           <at 0.7624 1>
           <at 1 0.6>
          ]}
         RedCurve {keys [
           <at 0 1 RCIM_Cubic>
           {arriveTangent=-1.3833, interpMode=RCIM_Cubic, leaveTangent=-1.3833, time=0.259, value=0.4467}
           <at 0.5034 0.3036>
           <at 0.7624 0.4467>
           <at 1 1>
          ]}
        }}
       CurveIndex=&Multiply_Float(A=$Particles.NormalizedAge, B=$Emitter.Audiogate)
      )
      Scale Factor=&MakeVector4_FromFloat(Float=&Multiply_Float(A=2.5, B=$Emitter.Audiogate))
     ))
    ScaleA=true
    ScaleRGB=true
   )
   &FN_Light_Attributes(
    Create Radius Attribute=true
    Radius=&Multiply_Float(A=180, B=&Multiply_Float(A=1, B=$Emitter.Audiogate))
   )
  ]
  rendererClasses [&NiagaraLightRendererProperties]
  renderers [{rendererClass=&NiagaraLightRendererProperties, rendererIndex=0}]
 }
 Waves {
  emitterSpawnScript [/SetVariables() {isSetParameters=true, module=SetVariables_#1}]
  emitterUpdateScript [
   &EmitterState(Life Cycle Mode=::NewEnumerator1, Loop Behavior=::NewEnumerator1, Loop Duration Mode=::NewEnumerator1, UseLoopDelay=false)
   &SpawnBurst_Instantaneous(Use Spawn Probability=false, UseLoopCountLimit=false)
   /SetVariables(Emitter.Audiogate:NiagaraFloat=&SmoothLerpOverTime_Float(
     Convergence Rate:NiagaraFloat=40
     Smooth Value:NiagaraFloat=$Emitter.IncomingAudio
     Target Value:NiagaraFloat=&ClampFloat(Float=&Multiply_Float(A=$NPC.FortniteNPC.EventAudio_Amplitude, B=1))
    )) {isSetParameters=true, module=SetVariables_#2}
  ]
  particleSpawnScript [
   &InitializeParticle(Color Mode=::NewEnumerator0, Mesh Scale Mode=::NewEnumerator3, UsePositionOffset=false, Write Material Random=true)
   &InitialMeshOrientation(Mesh Orientation Mode=::NewEnumerator2, Orientation Vector=<0 -1 0>, Use Rotation Vector=false)
   &DynamicMaterialParameters(
    Index 0 Param 2=2
    Module.Param0WriteEnabled=false
    Module.Param1WriteEnabled=true
    Module.Param2WriteEnabled=false
    Module.Param3WriteEnabled=false
   ) {module=DynamicMaterialParameters001}
  ]
  particleUpdateScript [
   &ParticleState(
    Kill Particles When Lifetime Has Elapsed=false
    Let Infinitely Lived Particles Die When Emitter Deactivates=true
    Loop Particles Lifetime=true
   )
   &SkeletalMeshLocation(Bone Sampling Mode=::NewEnumerator4, MaskSpawnGroup=false, Skeletal Mesh=$User.SkeletalMesh)
   &DynamicMaterialParameters(
    Index 0 Param 1=&Add_Float(
     A=&Multiply_Float(A=1, B=$Engine.Time)
     B=&Multiply_Float(A=1, B=$Emitter.Audiogate)
    )
    Module.Param0WriteEnabled=true
    Module.Param1WriteEnabled=false
    Module.Param2WriteEnabled=false
    Module.Param3WriteEnabled=false
   )
   &ScaleColor(
    Scale Alpha=&Multiply_Float(A=2, B=$Emitter.Audiogate)
    Scale RGB=&Multiply_Vector(
     A=<1 1 1>
     B=&VectorFromFloat(Value=&Add_Float(A=1, B=&Multiply_Float(A=2, B=$Emitter.Audiogate)))
    )
    ScaleA=true
    ScaleRGB=true
   )
  ]
  rendererClasses [&NiagaraMeshRendererProperties]
  renderers [{rendererClass=&NiagaraMeshRendererProperties, rendererIndex=0}]
 }
}
summary {
 systemName=NS_Cactus_Dancer_Male_IdleWaves
 emitters [
  {emitterName=Waves, rendererClasses [&NiagaraMeshRendererProperties]}
  {emitterName=IdleLight, rendererClasses [&NiagaraLightRendererProperties]}
 ]
 userVariables [
  User.bIsFrontEnd {type=NiagaraBool}
  User.bIsFrontEndPreview {type=NiagaraBool}
  User.SkeletalMesh {
   type=NiagaraDataInterfaceSkeletalMesh
   defaultValue {
    struct=NiagaraExt_VariableValue_DataInterface
    value {dataInterface=&self:NiagaraDataInterfaceSkeletalMesh_0}
   }
  }
 ]
}
